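# Conversation database for the Bountress NPC in the NPC Library, opening her dialogue when a player talks to her.
# A single entry point leads to a Speech node that picks a greeting from contextual lines gated by first-ever meeting, match conversation count or specific player outfits, with a default fallback.
# That node branches through link nodes to the shared NPC Library quests, basic services and Bountress-specific dialogue entry tags, none defined here.
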ConversationDatabase /NPCLibrary/NPCs/Bountress/NPCLibrary_Conversation_Bountress
entry_points [{tag=Conversation.Entry.Bountress, destinations [#1]}]
nodes [
 ConversationEntryPointNode_7 {
  id=#1
  class=ConversationEntryPointNode
  outputs [#2]
  properties {entryTag {tagName=Conversation.Entry.Bountress}}
 }
 Speech_C_4 {
  id=#2
  class=Speech_C
  outputs [#3, #4, #5]
  properties {
   nodeName=""
   generalConfig {
    defaultMessage=Need something? I'm working.
    contextualMessages [
     {
      message="Lucky for you, you're not on my list."
      contextRequirements [{requirement {class=FortControllerRequirement_IsFirstEverConversationWithNPC}}]
      priority {curve {rowName=Tandem.Speech.Priority.FirstEverInteraction}}
     }
     {
      message=Need something? I'm working.
      contextRequirements [{requirement {class=FortControllerRequirement_CheckNPCMatchConversationCount, properties {nPCUniqueIdTag {tagName=AISpawnerData.Type.Tandem.Bountress}}}}]
      priority {curve {rowName=None}}
     }
     {
      message=The hunt doesn't stop. Not even for you.
      contextRequirements [{requirement {class=FortControllerRequirement_CheckNPCMatchConversationCount, properties {nPCUniqueIdTag {tagName=AISpawnerData.Type.Tandem.Bountress}}}}]
      priority {curve {rowName=None}}
     }
     {
      message=Can I help you find something? Or someone?
      contextRequirements [{requirement {class=FortControllerRequirement_CheckNPCMatchConversationCount, properties {nPCUniqueIdTag {tagName=AISpawnerData.Type.Tandem.Bountress}}}}]
      priority {curve {rowName=None}}
     }
     {
      message=I'm looking for someone. Maybe you can help.
      contextRequirements [{requirement {class=FortControllerRequirement_CheckNPCMatchConversationCount, properties {nPCUniqueIdTag {tagName=AISpawnerData.Type.Tandem.Bountress}}}}]
      priority {curve {rowName=None}}
     }
     {
      message=Steer clear. These are my hunting grounds.
      contextRequirements [{requirement {
         class=FortControllerRequirement_HasCID
         properties {allowedCIds [/BRCosmetics/Athena/Items/Cosmetics/Characters/Character_SeleneCobra_Stride]}
        }}]
      priority {curve {rowName=Tandem.Speech.Priority.OutfitReaction}}
     }
     {
      message=I've been looking for you.
      contextRequirements [{requirement {
         class=FortControllerRequirement_HasCID
         properties {allowedCIds [/BRCosmetics/Athena/Items/Cosmetics/Characters/CID_217_Athena_Commando_M_Medic]}
        }}]
      priority {curve {rowName=Tandem.Speech.Priority.OutfitReaction}}
     }
     {
      message="I've tracked him this far, boss. Don't worry, I'll find him."
      contextRequirements [{requirement {
         class=FortControllerRequirement_HasCID
         properties {allowedCIds [/BRCosmetics/Athena/Items/Cosmetics/Characters/CID_216_Athena_Commando_F_Medic]}
        }}]
      priority {curve {rowName=Tandem.Speech.Priority.OutfitReaction}}
     }
    ]
   }
  }
 }
 ConversationLinkNode_2 {id=#5, class=ConversationLinkNode, properties {remoteEntryTag {tagName=Conversation.Entry.Bountress.Dialogue}}}
 ConversationLinkNode_0 {id=#4, class=ConversationLinkNode, properties {remoteEntryTag {tagName=Conversation.Entry.BasicServices}}}
 ConversationLinkNode_1 {id=#3, class=ConversationLinkNode, properties {remoteEntryTag {tagName=Conversation.Entry.NPCLibrary.Quests}}}
]
